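# Physics asset for the female medium Space Fighter cosmetic body, covering its dangling gear: front and back flaps, arm straps, pouches and a right leg pouch.
# Pelvis and thigh capsules form the body core; small base bodies join to the pelvis, or right thigh for the leg pouch, with free angular motion.
# Straps and pouches hang from their bases with tight angular limits, while box flaps chain through hinge bodies that swing on one axis only.
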PhysicsAsset /BRCosmetics/Characters/Player/Female/Medium/Bodies/F_MED_Space_Fighter/Meshes/F_MED_Space_Fighter_Physics
bodies {
 dyn_arm_strap_base_l [{shapeType=Capsule, length=0.505, radius=0.505, rotation=<rot 0 -90 0>}]
 dyn_arm_strap_base_r [{shapeType=Capsule, length=0.505, radius=0.505, rotation=<rot 0 -90 0>}]
 dyn_arm_strap_l [{shapeType=Capsule, length=2.25, radius=1.5, center=<1 0 0>, rotation=<rot 90 0 0>}]
 dyn_arm_strap_r [{shapeType=Capsule, length=2.25, radius=1.5, center=<-1 0 0>, rotation=<rot 90 0 0>}]
 dyn_bk_flap [{shapeType=Box, extentX=20.5, extentY=3.5, extentZ=20, center=<10.0147 0 0>}]
 dyn_bk_flap_hinge [{shapeType=Capsule, length=3.25, radius=2, center=<1.974 0 0>, rotation=<rot 90 0 0>}]
 dyn_bk_flap_hinge_base [{shapeType=Capsule, length=1, radius=1}]
 dyn_front_flap [{shapeType=Box, extentX=18.5, extentY=5.5, extentZ=19, center=<9.6111 0 0>}]
 dyn_front_flap_hinge [{shapeType=Capsule, length=1.5, radius=1.5, center=<1.8807 0 0>, rotation=<rot 90 0 0>}]
 dyn_front_flap_hinge_base [{shapeType=Capsule, length=1, radius=1}]
 dyn_leg_pouch_base_r [{shapeType=Capsule, length=0.505, radius=0.505, rotation=<rot 0 -90 0>}]
 dyn_leg_pouch_r [{shapeType=Capsule, length=7, radius=4.5, center=<1.3797 0 0>, rotation=<rot 90 0 0>}]
 dyn_pouch_a_base_l [{shapeType=Capsule, length=1, radius=1}]
 dyn_pouch_a_base_r [{shapeType=Capsule, length=1, radius=1}]
 dyn_pouch_a_l [{shapeType=Capsule, length=0, radius=3.5, center=<1.2894 0 0>, rotation=<rot 90 0 0>}]
 dyn_pouch_a_r [{shapeType=Capsule, length=5.25, radius=2.5, center=<-2.1322 0 0>, rotation=<rot 90 0 0>}]
 dyn_pouch_b_base_l [{shapeType=Capsule, length=1, radius=1}]
 dyn_pouch_b_base_r [{shapeType=Capsule, length=1, radius=1}]
 dyn_pouch_b_l [{shapeType=Capsule, length=5, radius=3, center=<1.8172 0 0>, rotation=<rot 90 0 0>}]
 dyn_pouch_b_r [{shapeType=Capsule, length=4.75, radius=2.75, center=<-1.6424 0 0>, rotation=<rot 90 0 -179.9999>}]
 dyn_pouch_c_base_l [{shapeType=Capsule, length=1, radius=1}]
 dyn_pouch_c_base_r [{shapeType=Capsule, length=1, radius=1}]
 dyn_pouch_c_l [{shapeType=Capsule, length=4, radius=2, center=<1.0378 0 0>, rotation=<rot 90 0 0>}]
 dyn_pouch_c_r [{shapeType=Capsule, length=4, radius=2, center=<-1.0378 0 0>, rotation=<rot 90 0 180>}]
 pelvis [{shapeType=Capsule, length=10, radius=9.5, center=<-3.3021 -1.8616 -0.0001>}]
 thigh_l [{shapeType=Capsule, length=29.25, radius=7.75, center=<-21.0009 0 -0.1887>, rotation=<rot 90 0 0>}]
 thigh_r [{shapeType=Capsule, length=29.25, radius=7.75, center=<21.0009 0 0.1887>, rotation=<rot 90 0 180>}]
}
constraints [
 {bone1Name=thigh_l, bone2Name=pelvis, swing1LimitDegrees=45, swing2LimitDegrees=45, twistLimitDegrees=45}
 {bone1Name=thigh_r, bone2Name=pelvis, swing1LimitDegrees=45, swing2LimitDegrees=45, twistLimitDegrees=45}
 {
  bone1Name=dyn_bk_flap_hinge_base
  bone2Name=pelvis
  swing1LimitDegrees=45
  swing1Motion=Free
  swing2LimitDegrees=45
  swing2Motion=Free
  twistLimitDegrees=45
  twistMotion=Free
 }
 {
  bone1Name=dyn_bk_flap_hinge
  bone2Name=dyn_bk_flap_hinge_base
  swing1LimitDegrees=5
  swing2LimitDegrees=45
  swing2Motion=Locked
  twistLimitDegrees=45
  twistMotion=Locked
 }
 {
  bone1Name=dyn_bk_flap
  bone2Name=dyn_bk_flap_hinge
  swing1LimitDegrees=60
  swing2LimitDegrees=45
  swing2Motion=Locked
  twistLimitDegrees=45
  twistMotion=Locked
 }
 {
  bone1Name=dyn_front_flap_hinge_base
  bone2Name=pelvis
  swing1LimitDegrees=45
  swing1Motion=Free
  swing2LimitDegrees=45
  swing2Motion=Free
  twistLimitDegrees=45
  twistMotion=Free
 }
 {
  bone1Name=dyn_front_flap_hinge
  bone2Name=dyn_front_flap_hinge_base
  swing1LimitDegrees=15
  swing2LimitDegrees=45
  swing2Motion=Locked
  twistLimitDegrees=45
  twistMotion=Locked
 }
 {
  bone1Name=dyn_front_flap
  bone2Name=dyn_front_flap_hinge
  swing1LimitDegrees=30
  swing2LimitDegrees=5
  swing2Motion=Locked
  twistLimitDegrees=45
  twistMotion=Locked
 }
 {
  bone1Name=dyn_arm_strap_base_r
  bone2Name=pelvis
  swing1LimitDegrees=45
  swing1Motion=Free
  swing2LimitDegrees=45
  swing2Motion=Free
  twistLimitDegrees=45
  twistMotion=Free
 }
 {
  bone1Name=dyn_arm_strap_base_l
  bone2Name=pelvis
  swing1LimitDegrees=45
  swing1Motion=Free
  swing2LimitDegrees=45
  swing2Motion=Free
  twistLimitDegrees=45
  twistMotion=Free
 }
 {bone1Name=dyn_arm_strap_l, bone2Name=dyn_arm_strap_base_l, swing1LimitDegrees=5, swing2LimitDegrees=15, twistLimitDegrees=5}
 {bone1Name=dyn_arm_strap_r, bone2Name=dyn_arm_strap_base_r, swing1LimitDegrees=5, swing2LimitDegrees=15, twistLimitDegrees=5}
 {
  bone1Name=dyn_pouch_a_base_l
  bone2Name=pelvis
  swing1LimitDegrees=45
  swing1Motion=Free
  swing2LimitDegrees=45
  swing2Motion=Free
  twistLimitDegrees=45
  twistMotion=Free
 }
 {
  bone1Name=dyn_pouch_b_base_l
  bone2Name=pelvis
  swing1LimitDegrees=45
  swing1Motion=Free
  swing2LimitDegrees=45
  swing2Motion=Free
  twistLimitDegrees=45
  twistMotion=Free
 }
 {
  bone1Name=dyn_pouch_c_base_l
  bone2Name=pelvis
  swing1LimitDegrees=45
  swing1Motion=Free
  swing2LimitDegrees=45
  swing2Motion=Free
  twistLimitDegrees=45
  twistMotion=Free
 }
 {
  bone1Name=dyn_pouch_a_base_r
  bone2Name=pelvis
  swing1LimitDegrees=45
  swing1Motion=Free
  swing2LimitDegrees=45
  swing2Motion=Free
  twistLimitDegrees=45
  twistMotion=Free
 }
 {
  bone1Name=dyn_pouch_b_base_r
  bone2Name=pelvis
  swing1LimitDegrees=45
  swing1Motion=Free
  swing2LimitDegrees=45
  swing2Motion=Free
  twistLimitDegrees=45
  twistMotion=Free
 }
 {
  bone1Name=dyn_pouch_c_base_r
  bone2Name=pelvis
  swing1LimitDegrees=45
  swing1Motion=Free
  swing2LimitDegrees=45
  swing2Motion=Free
  twistLimitDegrees=45
  twistMotion=Free
 }
 {bone1Name=dyn_pouch_a_l, bone2Name=dyn_pouch_a_base_l, swing1LimitDegrees=10, swing2LimitDegrees=2, twistLimitDegrees=5}
 {bone1Name=dyn_pouch_b_l, bone2Name=dyn_pouch_b_base_l, swing1LimitDegrees=10, swing2LimitDegrees=5, twistLimitDegrees=5}
 {bone1Name=dyn_pouch_b_r, bone2Name=dyn_pouch_b_base_r, swing1LimitDegrees=10, swing2LimitDegrees=5, twistLimitDegrees=5}
 {bone1Name=dyn_pouch_a_r, bone2Name=dyn_pouch_a_base_r, swing1LimitDegrees=10, swing2LimitDegrees=3, twistLimitDegrees=3}
 {
  bone1Name=dyn_leg_pouch_base_r
  bone2Name=thigh_r
  swing1LimitDegrees=45
  swing1Motion=Free
  swing2LimitDegrees=45
  swing2Motion=Free
  twistLimitDegrees=45
  twistMotion=Free
 }
 {bone1Name=dyn_leg_pouch_r, bone2Name=dyn_leg_pouch_base_r, swing1LimitDegrees=2, swing2LimitDegrees=2, twistLimitDegrees=2}
 {bone1Name=dyn_pouch_c_l, bone2Name=dyn_pouch_c_base_l, swing1LimitDegrees=5, swing2LimitDegrees=3, twistLimitDegrees=3}
 {bone1Name=dyn_pouch_c_r, bone2Name=dyn_pouch_c_base_r, swing1LimitDegrees=3, swing2LimitDegrees=3, twistLimitDegrees=3}
]
