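# Physics asset for the face accessory mesh of the female medium Space Fighter player cosmetic, defining hair bone physics.
# Small capsules cover the back, middle and front dyn_hair bones, alongside one large head capsule.
# Each hair body is constrained directly to the head rather than chained to other strands.
# Narrow swing and twist limits keep motion tight, with back hair swinging less than front and middle hair.
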PhysicsAsset /BRCosmetics/Characters/Player/Female/Medium/Bodies/F_MED_Space_Fighter/Meshes/Parts/F_MED_Space_Fighter_FaceAcc_Physics
bodies {
 dyn_hair_bk_a [{shapeType=Capsule, length=1.25, radius=0.75, center=<-0.6394 0.0001 0>, rotation=<rot 90 0 0>}]
 dyn_hair_bk_b [{shapeType=Capsule, length=1, radius=0.75, center=<-1.0912 0 0>, rotation=<rot 90 0 0>}]
 dyn_hair_bk_c [{shapeType=Capsule, length=1.25, radius=0.75, center=<-0.7682 0 0>, rotation=<rot 90 0 0>}]
 dyn_hair_bk_d [{shapeType=Capsule, length=1.25, radius=0.75, center=<-0.5905 0 0>, rotation=<rot 90 0 0>}]
 dyn_hair_bk_e [{shapeType=Capsule, length=1.25, radius=0.75, center=<-0.5856 0 0>, rotation=<rot 90 0 0>}]
 dyn_hair_bk_f [{shapeType=Capsule, length=1.25, radius=0.75, center=<-0.7227 0 0>, rotation=<rot 90 0 0>}]
 dyn_hair_bk_g [{shapeType=Capsule, length=1.25, radius=0.75, center=<-0.5413 0 0>, rotation=<rot 90 0 0>}]
 dyn_hair_bk_h [{shapeType=Capsule, length=1, radius=0.75, center=<-0.7471 0 0>, rotation=<rot 90 0 0>}]
 dyn_hair_fr_a [{shapeType=Capsule, length=2.5, radius=1.5, center=<-1.9097 0 0>, rotation=<rot 90 0 0>}]
 dyn_hair_fr_b [{shapeType=Capsule, length=3.25, radius=1, center=<-1.9882 -0.0001 0>, rotation=<rot 90 0 0>}]
 dyn_hair_fr_c [{shapeType=Capsule, length=1.75, radius=1, center=<-1.2415 0 0>, rotation=<rot 90 0 0>}]
 dyn_hair_fr_d [{shapeType=Capsule, length=2, radius=1, center=<-1.6107 0 0>, rotation=<rot 90 0 0>}]
 dyn_hair_fr_e [{shapeType=Capsule, length=1.25, radius=0.75, center=<-1.1005 0 0>, rotation=<rot 90 0 0>}]
 dyn_hair_fr_f [{shapeType=Capsule, length=1.25, radius=0.75, center=<-1.0104 0 0>, rotation=<rot 90 0 0>}]
 dyn_hair_fr_g [{shapeType=Capsule, length=1.25, radius=0.75, center=<-0.7803 0 0>, rotation=<rot 90 0 0>}]
 dyn_hair_fr_h [{shapeType=Capsule, length=1.25, radius=0.75, center=<-1.3064 0 0>, rotation=<rot 90 0 0>}]
 dyn_hair_fr_i [{shapeType=Capsule, length=1.25, radius=0.75, center=<-0.7357 0 0>, rotation=<rot 90 0 0>}]
 dyn_hair_fr_j [{shapeType=Capsule, length=1.25, radius=0.75, center=<-1.2173 0 0>, rotation=<rot 90 0 0>}]
 dyn_hair_md_a [{shapeType=Capsule, length=1.25, radius=0.75, center=<-0.6056 0 0>, rotation=<rot 90 0 0>}]
 dyn_hair_md_b [{shapeType=Capsule, length=2.25, radius=1, center=<-1.8269 0 0>, rotation=<rot 90 0 0>}]
 head [{shapeType=Capsule, length=4.75, radius=6.25, center=<2.1401 2.9408 0.0001>, rotation=<rot 90 0 0>}]
}
constraints [
 {bone1Name=dyn_hair_bk_a, bone2Name=head, swing1LimitDegrees=10, swing2LimitDegrees=10, twistLimitDegrees=5}
 {bone1Name=dyn_hair_bk_b, bone2Name=head, swing1LimitDegrees=10, swing2LimitDegrees=10, twistLimitDegrees=5}
 {bone1Name=dyn_hair_bk_c, bone2Name=head, swing1LimitDegrees=10, swing2LimitDegrees=10, twistLimitDegrees=5}
 {bone1Name=dyn_hair_bk_f, bone2Name=head, swing1LimitDegrees=10, swing2LimitDegrees=10, twistLimitDegrees=5}
 {bone1Name=dyn_hair_bk_g, bone2Name=head, swing1LimitDegrees=10, swing2LimitDegrees=10, twistLimitDegrees=5}
 {bone1Name=dyn_hair_bk_e, bone2Name=head, swing1LimitDegrees=10, swing2LimitDegrees=10, twistLimitDegrees=5}
 {bone1Name=dyn_hair_bk_d, bone2Name=head, swing1LimitDegrees=10, swing2LimitDegrees=10, twistLimitDegrees=5}
 {bone1Name=dyn_hair_md_a, bone2Name=head, swing1LimitDegrees=15, swing2LimitDegrees=15, twistLimitDegrees=5}
 {bone1Name=dyn_hair_md_b, bone2Name=head, swing1LimitDegrees=15, swing2LimitDegrees=15, twistLimitDegrees=5}
 {bone1Name=dyn_hair_fr_a, bone2Name=head, swing1LimitDegrees=15, swing2LimitDegrees=15, twistLimitDegrees=5}
 {bone1Name=dyn_hair_fr_b, bone2Name=head, swing1LimitDegrees=15, swing2LimitDegrees=15, twistLimitDegrees=5}
 {bone1Name=dyn_hair_fr_c, bone2Name=head, swing1LimitDegrees=15, swing2LimitDegrees=15, twistLimitDegrees=5}
 {bone1Name=dyn_hair_fr_d, bone2Name=head, swing1LimitDegrees=15, swing2LimitDegrees=15, twistLimitDegrees=5}
 {bone1Name=dyn_hair_fr_e, bone2Name=head, swing1LimitDegrees=15, swing2LimitDegrees=15, twistLimitDegrees=5}
 {bone1Name=dyn_hair_fr_f, bone2Name=head, swing1LimitDegrees=15, swing2LimitDegrees=15, twistLimitDegrees=5}
 {bone1Name=dyn_hair_fr_g, bone2Name=head, swing1LimitDegrees=15, swing2LimitDegrees=15, twistLimitDegrees=5}
 {bone1Name=dyn_hair_fr_h, bone2Name=head, swing1LimitDegrees=15, swing2LimitDegrees=15, twistLimitDegrees=5}
 {bone1Name=dyn_hair_fr_i, bone2Name=head, swing1LimitDegrees=15, swing2LimitDegrees=15, twistLimitDegrees=5}
 {bone1Name=dyn_hair_fr_j, bone2Name=head, swing1LimitDegrees=15, swing2LimitDegrees=15, twistLimitDegrees=5}
 {bone1Name=dyn_hair_bk_h, bone2Name=head, swing1LimitDegrees=10, swing2LimitDegrees=10, twistLimitDegrees=5}
]
